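# Anim Blueprint for the face accessory of the F_MED_NuggetReal_Hat player cosmetic, copying its pose from the attached parent mesh.
# The event graph reads head rotation relative to spine_05 from the owning FortPlayerPawn's mesh and remaps negative yaw into a small clamped HatRotate yaw.
# An additive bone-space Modify Bone chain applies HatRotate to the hat bone, but it is not wired to Output Pose, which takes the copied pose directly.
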AnimBlueprint /BRCosmetics/Characters/Player/Female/Medium/Bodies/F_MED_NuggetReal_Hat/Meshes/Parts/F_MED_NuggetReal_Hat_FaceAcc_AnimBP
target_skeleton=&/F_MED_NuggetReal_Hat_FaceAcc_Skeleton
graphs [
 AnimGraph {nodes [
   Component To Local(ComponentPose="Transform (Modify) Bone - Bone: hat"(ComponentPose=Local To Component(), Alpha=1, Translation="0,0,0", Rotation=Get HatRotate() -> (HatRotate:Rotator), Scale="1,1,1") {rotationMode=BMM_Additive, rotationSpace=BCS_BoneSpace, boneToModify {boneName=hat}})
   Output Pose(Result=Copy Pose From Mesh() {bUseAttachedParent=true}) {layerGroup=None}
  ]}
 EventGraph {nodes [
   Event BlueprintUpdateAnimation()
   > n1 = Cast To FortPlayerPawn(Object=TryGetPawnOwner())
   n2 = Get Mesh(self=@n1)
   n3 = BreakRotator(InRot=BreakTransform(InTransform=GetDeltaTransformFromRefPose(self=@n2, BoneName=head, BaseName=spine_05)).Rotation)
   > Set HatRotate(HatRotate:Rotator=MakeRotator(Yaw=MapRangeClamped(Value=@n3.Yaw, InRangeA=-30, InRangeB=0, OutRangeA=10, OutRangeB=0))) -> (Output_Get:Rotator)
  ]}
]
variables [HatRotate {type=Rotator}]
